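# Physics asset for the face-accessory part of the male medium EthicalNogginHood player cosmetic, covering its hood bones.
# It holds two small capsule bodies, one on MSTR_hood_1 and one offset along its bone on hood_1.
# A single constraint joins the two with limited swing on both axes and a tighter twist, restricting how far the hood can move relative to its master bone.
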PhysicsAsset /BRCosmetics/Characters/Player/Male/Medium/Bodies/M_MED_EthicalNogginHood/Meshes/Parts/M_MED_EthicalNogginHood_FaceAcc_Physics
bodies {
 MSTR_hood_1 [{shapeType=Capsule, length=1, radius=1, shapeName=MSTR_hood_1_capsule}]
 hood_1 [{shapeType=Capsule, length=1, radius=1, shapeName=hood_1_capsule, center=<6.8248 0 0>}]
}
constraints [{bone1Name=MSTR_hood_1, bone2Name=hood_1, swing1LimitDegrees=25, swing2LimitDegrees=15, twistLimitDegrees=5}]
